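# UI-domain material with Modulate blending that shades whatever lies behind it using a soft grayscale gradient.
# Fill-scaled UVs drive an eased linear gradient and an eased radial gradient, the latter offset horizontally and stretched about its center.
# A fixed mix of both, weighted toward the radial, interpolates between dark gray and white, so white areas leave the backdrop unchanged.
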Material /FortUILibrary/UIKit/ModulateBackground/Materials/M_UI_ModulateBackground
graph {
 properties {
  bAutomaticallySetUsageInEditor=false
  bUsedWithStaticMesh=false
  blendMode=BLEND_Modulate
  materialDomain=MD_UI
  usedShadingModels=Default Lit
 }
 nodes [
  n1 = MaterialFunctionCall(UV=TextureCoordinate()) {materialFunction=&MF_UI_UVScaling}
  n2 = MaterialFunctionCall(UVs=@n1.Scaled To Fill) {materialFunction=&MF_UI_LinearGradient}
  n3 = AppendVector(A=Constant() {r=-0.1}, B=Constant())
  n4 = Constant() {r=1.3}
  n5 = Constant() {r=1.1}
  n6 = MaterialFunctionCall(UV=MaterialFunctionCall(UVs=Add(A=@n1.Scaled To Fill, B=@n3), Texture Scale=AppendVector(A=@n4, B=@n5)) {materialFunction=&ScaleUVsByCenter}.UVs) {materialFunction=&MF_UI_RadialGradient}
  LinearInterpolate(
   A=Constant3Vector() {constant=<rgba 0.1094 0.1094 0.1094 0>}
   B=Constant3Vector() {constant=<rgba 1 1 1 0>}
   Alpha=LinearInterpolate(
    A=MaterialFunctionCall("Value (0-1)"=@n2.Result) {materialFunction=&MF_UI_EaseCurves}.EaseInOutCubic
    B=MaterialFunctionCall("Value (0-1)"=@n6.Smooth) {materialFunction=&MF_UI_EaseCurves}.EaseOutSine
   ) {constAlpha=0.65}
  )
 ]
}
